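# Physics asset for the Pug Militia backpack cosmetic, defining bodies and joints for its dangling parts.
# Each body is a single box on the pack root, a dyn_ bone, two handle strings, a top zipper or the left and right grenades.
# Angle-limited constraints join the dyn_ bone, strings and zipper to the pack, with tight twist throughout.
# Each grenade swings from its own base body, and neither base has a constraint to the pack.
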PhysicsAsset /BRCosmetics/Accessories/FORT_Backpacks/Backpack_M_MED_Pug_Militia/Meshes/M_MED_Pug_Militia_Physics
bodies {
 dyn_bone [{shapeType=Box, extentX=4.8866, extentY=16.9387, extentZ=5.3323, center=<1.8418 0 0>}]
 dyn_granade_l_base [{shapeType=Box, extentX=1.25, extentY=1.25, extentZ=1.25}]
 dyn_granade_l_parent [{shapeType=Box, extentX=8.8369, extentY=5.7838, extentZ=6.8237, center=<2.9413 0 0.4134>}]
 dyn_granade_r_base [{shapeType=Box, extentX=1.125, extentY=1.125, extentZ=1.125}]
 dyn_granade_r_parent [{shapeType=Box, extentX=8.8291, extentY=5.7683, extentZ=6.8237, center=<2.9702 0 0>}]
 dyn_handle_string_01 [{shapeType=Box, extentX=3.5488, extentY=2.8296, extentZ=1.3187, center=<1.3463 0 0>}]
 dyn_handle_string_02 [{shapeType=Box, extentX=3.9605, extentY=2.6884, extentZ=0.9375, center=<1.7001 -0.0001 0>}]
 dyn_zipper_top_parent [{shapeType=Box, extentX=2.7194, extentY=1.875, extentZ=1.0625, center=<1.1718 0 0>}]
 pack [{shapeType=Box, extentX=3, extentY=3, extentZ=3}]
}
constraints [
 {bone1Name=dyn_handle_string_01, bone2Name=pack, swing1LimitDegrees=20, swing2LimitDegrees=20, twistLimitDegrees=3}
 {bone1Name=dyn_handle_string_02, bone2Name=pack, swing1LimitDegrees=20, swing2LimitDegrees=20, twistLimitDegrees=3}
 {bone1Name=dyn_granade_l_parent, bone2Name=dyn_granade_l_base, swing1LimitDegrees=25, swing2LimitDegrees=25, twistLimitDegrees=3}
 {bone1Name=dyn_granade_r_parent, bone2Name=dyn_granade_r_base, swing1LimitDegrees=25, swing2LimitDegrees=25, twistLimitDegrees=3}
 {bone1Name=dyn_bone, bone2Name=pack, swing1LimitDegrees=5, swing2LimitDegrees=5, twistLimitDegrees=3}
 {bone1Name=dyn_zipper_top_parent, bone2Name=pack, swing1LimitDegrees=30, swing2LimitDegrees=30, twistLimitDegrees=3}
]
